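; Camera modifier for the Smartgun stacking lock-on that applies a screen post process to the auto-aim victim.
; Each update casts the camera owner's controlled pawn to FortPawn, setting IsActive false if the cast fails or the pawn is DBNO or dead.
; Otherwise, BlendWeight eases toward 1 when active or 0 when inactive via FInterpTo and is returned with VictimPostProcess.
; The deactivation branches skip both the blend and that return.
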(fn BlueprintModifyPostProcess
  (params (DeltaTime :type Float%20%28single-precision%29))
  (bind _asfort_pawn (Utilities|Casting|CastToFortPawn (Pawn|GetControlledPawn (Game|Player|GetOwningPlayerController (Variables|CameraModifier|GetCameraOwner))))
    (:then
      (if (or (Health|IsDBNO _asfort_pawn) (FortPawn|IsDead _asfort_pawn))
        (Variables|Default|SetIsActive false)
        (else
          (Variables|Default|SetBlendWeight (Math|Interpolation|FInterpTo (Variables|Default|GetBlendWeight) (Math|Float|SelectFloat 1.0 0.0 (Variables|Default|GetIsActive)) (Variables|GetDeltaTime) 4.0))
          (return (Variables|Default|GetBlendWeight) (Variables|Default|GetVictimPostProcess)))))
    (:CastFailed
      (Variables|Default|SetIsActive false))))
